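# Physics asset for the glute chain and two kunai on the F_MED_LilacLeather female player body.
# Capsule bodies cover the chain bones, both kunai, the pelvis and the right thigh.
# Constraints hang Base_chain_glute from the pelvis, then Dyn_chain_glute, then Base_chain_bk, with Kunai_BK branching off the base and Kunai_FR off the dynamic link.
# Twist is locked on both kunai; the pelvis-to-chain joint locks twist and swing 2.
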PhysicsAsset /BRCosmetics/Characters/Player/Female/Medium/Bodies/F_MED_LilacLeather/Meshes/F_MED_LilacLeather_Physics_CHAIN1
bodies {
 Base_chain_bk [{shapeType=Capsule, length=0.505, radius=0.505, shapeName=Base_chain_bk_capsule, rotation=<rot 0 -90 0>}]
 Base_chain_glute [{shapeType=Capsule, length=6.505, radius=3.005, shapeName=Base_chain_glute_capsule, center=<5.9605 -2.9211 0>, rotation=<rot 0 -90 0>}]
 Dyn_chain_glute [{shapeType=Capsule, length=6.005, radius=2.505, shapeName=Dyn_chain_glute_capsule, center=<0 -1.1458 0>, rotation=<rot 0 -90 0>}]
 Kunai_BK [{
   shapeType=Capsule
   length=6.2424
   radius=2.7955
   shapeName=Kunai_BK_capsule
   center=<5.9801 -0.6396 -1.4613>
   rotation=<rot 0 -88.7057 -94.4552>
  }]
 Kunai_FR [{
   shapeType=Capsule
   length=6.2176
   radius=2.7973
   shapeName=Kunai_FR_capsule
   center=<5.1545 -0.0653 -1.528>
   rotation=<rot 0 -90.2864 -97.602>
  }]
 pelvis [{shapeType=Capsule, length=11.9255, radius=10.1811, center=<-1.3453 -0.9443 0>, rotation=<rot 0 0 -0.0243>}]
 thigh_r [{shapeType=Capsule, length=21.5802, radius=9.3821, center=<16.4898 1.5295 -1.7103>, rotation=<rot 0 -92.5786 -87.7498>}]
}
constraints [
 {bone1Name=thigh_r, bone2Name=pelvis, swing1LimitDegrees=5, swing2LimitDegrees=5, twistLimitDegrees=5}
 {bone1Name=Dyn_chain_glute, bone2Name=Base_chain_glute, swing1LimitDegrees=30, swing2LimitDegrees=45, twistLimitDegrees=15}
 {
  bone1Name=Base_chain_glute
  bone2Name=pelvis
  swing1LimitDegrees=25
  swing2LimitDegrees=45
  swing2Motion=Locked
  twistLimitDegrees=45
  twistMotion=Locked
 }
 {bone1Name=Base_chain_bk, bone2Name=Dyn_chain_glute, swing1LimitDegrees=45, swing2LimitDegrees=45, twistLimitDegrees=15}
 {bone1Name=Kunai_FR, bone2Name=Dyn_chain_glute, swing1LimitDegrees=45, swing2LimitDegrees=45, twistLimitDegrees=45, twistMotion=Locked}
 {bone1Name=Kunai_BK, bone2Name=Base_chain_glute, swing1LimitDegrees=25, swing2LimitDegrees=25, twistLimitDegrees=45, twistMotion=Locked}
]
