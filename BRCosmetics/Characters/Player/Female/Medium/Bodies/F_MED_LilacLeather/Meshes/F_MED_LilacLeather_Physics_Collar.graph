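# Physics asset for the dynamic jacket collar on the medium female LilacLeather player body, built entirely from capsule bodies.
# Left and right three-bone collar chains are constrained to spine_05, with capsules on the spine and both upper arms giving the collar body shapes to collide with.
# The sides are tuned asymmetrically: the left locks its middle joint and tip twist, while the right uses angular limits throughout.
# Upper-arm constraints leave swing and twist free.
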PhysicsAsset /BRCosmetics/Characters/Player/Female/Medium/Bodies/F_MED_LilacLeather/Meshes/F_MED_LilacLeather_Physics_Collar
bodies {
 Base_collar_L [{shapeType=Capsule, length=0.505, radius=0.505, shapeName=Base_collar_L_capsule, rotation=<rot 0 -90 0>}]
 Dyn_collar_L_01 [{
   shapeType=Capsule
   length=7.1347
   radius=2.5708
   shapeName=Dyn_collar_L_01_capsule
   center=<2.5057 0.0979 1.1194>
   rotation=<rot -54.6175 -19.2682 -84.4399>
  }]
 Dyn_collar_L_02 [{
   shapeType=Capsule
   length=5.816
   radius=3.0664
   shapeName=Dyn_collar_L_02_capsule
   center=<2.3885 -1.5287 0.8812>
   rotation=<rot -0.0001 0.0001 -98.6036>
  }]
 Dyn_collar_R_ [{shapeType=Capsule, length=0.505, radius=0.505, shapeName=Dyn_collar_R__capsule, rotation=<rot 0 -90 0>}]
 Dyn_collar_R_01 [{
   shapeType=Capsule
   length=4.5
   radius=3.5
   shapeName=Dyn_collar_R_01_capsule
   center=<-2.0807 2.2854 -2.8795>
   rotation=<rot 13.7338 153.1055 -109.7247>
  }]
 Dyn_collar_R_02 [{
   shapeType=Capsule
   length=6.5
   radius=3
   shapeName=Dyn_collar_R_02_capsule
   center=<-2.327 4.0733 -1.4297>
   rotation=<rot 0.0001 -171.9961 -76.3981>
  }]
 spine_05 [{
   shapeType=Capsule
   length=10.8448
   radius=4.259
   shapeName=spine_05_capsule
   center=<7.1235 -1.2032 0.8749>
   rotation=<rot 0 178.3675 9.5957>
  }]
 upperarm_l [{
   shapeType=Capsule
   length=13.7345
   radius=4.6534
   shapeName=upperarm_l_capsule
   center=<7.386 -0.1411 0.3109>
   rotation=<rot 13.0646 87.3763 -87.9972>
  }]
 upperarm_r [{shapeType=Capsule, length=13, radius=5, shapeName=upperarm_r_capsule, center=<-7.386 0.1411 -0.3109>, rotation=<rot 13.1359 84.9525 91.4533>}]
}
constraints [
 {bone1Name=Base_collar_L, bone2Name=spine_05, swing1LimitDegrees=20, swing2LimitDegrees=20, twistLimitDegrees=15}
 {
  bone1Name=Dyn_collar_L_01
  bone2Name=Base_collar_L
  swing1LimitDegrees=10
  swing1Motion=Locked
  swing2LimitDegrees=10
  swing2Motion=Locked
  twistLimitDegrees=45
  twistMotion=Locked
 }
 {bone1Name=Dyn_collar_L_02, bone2Name=Dyn_collar_L_01, swing1LimitDegrees=10, swing2LimitDegrees=10, twistLimitDegrees=45, twistMotion=Locked}
 {
  bone1Name=upperarm_l
  bone2Name=spine_05
  swing1LimitDegrees=45
  swing1Motion=Free
  swing2LimitDegrees=45
  swing2Motion=Free
  twistLimitDegrees=45
  twistMotion=Free
 }
 {
  bone1Name=upperarm_r
  bone2Name=spine_05
  swing1LimitDegrees=45
  swing1Motion=Free
  swing2LimitDegrees=45
  swing2Motion=Free
  twistLimitDegrees=45
  twistMotion=Free
 }
 {bone1Name=Dyn_collar_R_, bone2Name=spine_05, swing1LimitDegrees=5, swing2LimitDegrees=5, twistLimitDegrees=10}
 {bone1Name=Dyn_collar_R_01, bone2Name=Dyn_collar_R_, swing1LimitDegrees=5, swing2LimitDegrees=3, twistLimitDegrees=3}
 {bone1Name=Dyn_collar_R_02, bone2Name=Dyn_collar_R_01, swing1LimitDegrees=11, swing2LimitDegrees=11, twistLimitDegrees=11}
]
