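# Post-process style Animation Blueprint for the F_MED_LilacLeather female body, adding secondary motion to its chains, collar, jacket, zipper and pendant.
# Copied parent-mesh pose feeds a Control Rig, then RigidBody and AnimDynamics nodes, with weighted constraints on pendant and zipper bases.
# Two chain constraints sit disconnected from the output.
# Events set front-end simulation variables and hang-glider state, reset dynamics on customization load, and toggle tail visibility for long backblings.
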AnimBlueprint /BRCosmetics/Characters/Player/Female/Medium/Bodies/F_MED_LilacLeather/Meshes/F_MED_LilacLeather_AnimBP
target_skeleton=&/F_MED_LilacLeather_Skeleton
graphs [
 AnimGraph {nodes [
   n1 = "Constraint - Dyn_chain_bk_01 (PPPPPPPPPPP)"!(
    ComponentPose
    Alpha=1
    ConstraintWeights_0=0.0003
    ConstraintWeights_1:float=0.0004
    ConstraintWeights_2:float=0.0137
    ConstraintWeights_3:float=0.0059
    ConstraintWeights_4:float=0.5708
    ConstraintWeights_5:float=0
    ConstraintWeights_6:float=0.0047
    ConstraintWeights_7:float=0
    ConstraintWeights_8:float=0
    ConstraintWeights_9:float=0.0253
    ConstraintWeights_10:float=0.3788
   ) -> (Pose) {
    alphaBoolBlend {blendOption=Linear}
    boneToModify {boneName=Dyn_chain_bk_01}
    constraintSetup [
     {targetBone {boneName=spine_04}}
     {targetBone {boneName=spine_03}}
     {targetBone {boneName=spine_02}}
     {targetBone {boneName=spine_01}}
     {targetBone {boneName=pelvis}}
     {targetBone {boneName=thigh_twist_01_l}}
     {targetBone {boneName=thigh_twist_01_r}}
     {targetBone {boneName=thigh_r}}
     {targetBone {boneName=deform_glute_l}}
     {targetBone {boneName=deform_glute_r}}
     {targetBone {boneName=Dyn_chain_bk_01}}
    ]
    constraintWeights [1,1,1,1,1,1,1,1,1,1,1]
   }
   "Constraint - Dyn_chain_fr_01 (PPPPPPP)"!(
    ComponentPose=@n1
    Alpha=1
    ConstraintWeights_0=0
    ConstraintWeights_1:float=0.6547
    ConstraintWeights_2:float=0.0281
    ConstraintWeights_3:float=0.0002
    ConstraintWeights_4:float=0.0029
    ConstraintWeights_5:float=0.005
    ConstraintWeights_6:float=0.309
   ) -> (Pose) {
    alphaBoolBlend {blendOption=Linear}
    boneToModify {boneName=Dyn_chain_fr_01}
    constraintSetup [
     {targetBone {boneName=spine_02}}
     {targetBone {boneName=pelvis}}
     {targetBone {boneName=thigh_twist_01_r}}
     {targetBone {boneName=thigh_r}}
     {targetBone {boneName=deform_glute_r}}
     {targetBone {boneName=deform_groin_r}}
     {targetBone {boneName=Dyn_chain_fr_01}}
    ]
    constraintWeights [1,1,1,1,1,1,1]
   }
   n2 = Control Rig!(
    Source=Copy Pose From Mesh() {bUseAttachedParent=true}
    Alpha
    HangTwo arms:double=Get HangTwo arms() -> (HangTwo arms:double)
    DBNO:bool=Get bIsDBNO()
    RBAN:bool=Get bRigidBodyEnabled()
    ISPlayingEmote:bool=Get bIsPlayingEmote()
    IsFrontEnd:bool=Get bIsInFrontEnd()
   ) -> (Pose) {
    alphaBoolBlend {blendOption=Linear}
    controlRigAssetReference {blueprintRigClass=&/F_MED_LilacLeather_CtrlRig.F_MED_LilacLeather_CtrlRig_C}
   }
   n3 = "Constraint - Base_Pendant (PPPPP)"!(
    ComponentPose=Local To Component(LocalPose=@n2)
    Alpha=1
    ConstraintWeights_0=0.7213
    ConstraintWeights_1:float=0
    ConstraintWeights_2:float=0.1992
    ConstraintWeights_3:float=0.0546
    ConstraintWeights_4:float=0.0248
   ) -> (Pose) {
    alphaBoolBlend {blendOption=Linear}
    boneToModify {boneName=Base_Pendant}
    constraintSetup [
     {targetBone {boneName=spine_05}}
     {targetBone {boneName=clavicle_r}}
     {targetBone {boneName=neck_01}}
     {targetBone {boneName=neck_02}}
     {targetBone {boneName=FACIAL_C_Jawline}}
    ]
    constraintWeights [1,1,1,1,1]
   }
   n4 = RigidBody(ComponentPose=@n3, Alpha=0.5) {
    bOverrideWorldGravity=true
    overridePhysicsAsset=&/F_MED_LilacLeather_Physics_CHAIN1
    simulationSpace=BaseBoneSpace
    alphaBoolBlend {blendOption=Linear}
    baseBoneRef {boneName=pelvis}
    overrideWorldGravity {z=-980}
    simSpaceSettings {maxAngularAcceleration=8, maxAngularVelocity=30, maxLinearAcceleration=700, maxLinearVelocity=700, worldAlpha=1}
   }
   n5 = Anim Dynamics - Bone: Base_chain_glute(ComponentPose=@n4, Alpha=1, ExternalForce="0,0,0") {
    angularDampingOverride=0.999
    angularSpringConstant=333
    bAngularSpring=true
    bOverrideAngularDamping=true
    bUseGravityOverride=true
    simulationSpace=RootRelative
    alphaBoolBlend {blendOption=Linear}
    boundBone {boneName=Base_chain_glute}
    componentAppliedLinearAccClamp=<1800 1800 1800>
    componentLinearAccScale=<0.7 0.7 0.7>
    componentLinearVelScale=<0.7 0.7 0.7>
    gravityOverride {z=-980}
    physicsBodyDefinitions [{
      boundBone {boneName=Base_chain_glute}
      boxExtents=<10 10 10>
      constraintSetup {angularLimitsMax=<0 5 3>, angularLimitsMin=<0 -5 -33>, angularTarget=<1 0 0>}
      localJointOffset=<0.7 0 0>
     }]
    relativeSpaceBone {boneName=pelvis}
   }
   n6 = RigidBody(ComponentPose=@n5) {
    bOverrideWorldGravity=true
    overridePhysicsAsset=&/F_MED_LilacLeather_Physics_Collar
    simulationSpace=BaseBoneSpace
    alphaBoolBlend {blendOption=Linear}
    baseBoneRef {boneName=spine_05}
    overrideWorldGravity {z=-980}
    simSpaceSettings {maxAngularAcceleration=8, maxAngularVelocity=30, maxLinearAcceleration=700, maxLinearVelocity=700, worldAlpha=1}
   }
   n7 = Anim Dynamics - Bone: Dyn_jacket_r!(SimulationSpace:AnimPhysSimSpaceType Enum, ComponentPose=@n6, Alpha=1, GravityOverride:Vector="0,0,-980", ExternalForce="0,0,0") -> (Pose) {
    angularDampingOverride=1
    angularSpringConstant=500
    bAngularSpring=true
    bGravityOverrideInSimSpace=true
    bOverrideAngularDamping=true
    bUseGravityOverride=true
    bUseSphericalLimits=true
    simulationSpace=RootRelative
    alphaBoolBlend {blendOption=Linear}
    alphaScaleBiasClamp {bClampResult=true}
    boundBone {boneName=Dyn_jacket_r}
    componentAppliedLinearAccClamp=<5000 5000 5000>
    componentLinearAccScale=<0.1 0.1 0.1>
    componentLinearVelScale=<0.1 0.1 0.1>
    gravityOverride {z=-980}
    physicsBodyDefinitions [{
      collisionType=CustomSphere
      sphereCollisionRadius=3
      boundBone {boneName=Dyn_jacket_r}
      boxExtents=<11 11 11>
      constraintSetup {coneAngle=1, angularLimitsMax=<5 20 13>, angularLimitsMin=<-5 -20 -13>, angularTarget=<1 0 0>}
      localJointOffset=<0.6 0 0>
     }]
    relativeSpaceBone {boneName=spine_04}
    sphericalLimits [{limitRadius=9, drivingBone {boneName=spine_03}, sphereLocalOffset=<0 0 0>}]
   }
   n8 = Anim Dynamics - Bone: Dyn_jacket_l_01!(SimulationSpace:AnimPhysSimSpaceType Enum, ComponentPose=@n7, Alpha=1, GravityOverride:Vector="0,0,-980", ExternalForce="0,0,0") -> (Pose) {
    angularDampingOverride=1
    angularSpringConstant=500
    bAngularSpring=true
    bGravityOverrideInSimSpace=true
    bOverrideAngularDamping=true
    bUseGravityOverride=true
    bUseSphericalLimits=true
    simulationSpace=RootRelative
    alphaBoolBlend {blendOption=Linear}
    alphaScaleBiasClamp {bClampResult=true}
    boundBone {boneName=Dyn_jacket_l_01}
    componentAppliedLinearAccClamp=<5000 5000 5000>
    componentLinearAccScale=<0.1 0.1 0.1>
    componentLinearVelScale=<0.1 0.1 0.1>
    gravityOverride {z=-980}
    physicsBodyDefinitions [{
      collisionType=CustomSphere
      sphereCollisionRadius=3
      boundBone {boneName=Dyn_jacket_l_01}
      boxExtents=<11 11 11>
      constraintSetup {coneAngle=1, angularLimitsMax=<5 20 20>, angularLimitsMin=<-5 -20 -10>, angularTarget=<1 0 0>}
      localJointOffset=<0.6 0 0>
     }]
    relativeSpaceBone {boneName=spine_04}
    sphericalLimits [{limitRadius=9.5, drivingBone {boneName=spine_03}, sphereLocalOffset=<0 0 0>}]
   }
   n9 = Anim Dynamics - Bone: Dyn_zipper!(SimulationSpace:AnimPhysSimSpaceType Enum, ComponentPose=@n8, Alpha=1, GravityOverride:Vector="0,0,-980", ExternalForce="0,0,0") -> (Pose) {
    angularDampingOverride=1
    angularSpringConstant=500
    bAngularSpring=true
    bOverrideAngularDamping=true
    bUseGravityOverride=true
    gravityScale=0.35
    simulationSpace=RootRelative
    alphaBoolBlend {blendOption=Linear}
    boundBone {boneName=Dyn_zipper}
    componentAppliedLinearAccClamp=<6000 6000 6000>
    componentLinearAccScale=<1 1 1>
    componentLinearVelScale=<1 1 1>
    gravityOverride {z=-980}
    physicsBodyDefinitions [{
      boundBone {boneName=Dyn_zipper}
      boxExtents=<10 10 10>
      constraintSetup {angularLimitsMax=<15 15 15>, angularLimitsMin=<-15 -75 -15>, angularTarget=<1 0 0>}
      localJointOffset=<1.2 0 0>
     }]
   }
   n10 = Anim Dynamics - Bone: Dyn_pendant!(SimulationSpace:AnimPhysSimSpaceType Enum, ComponentPose=@n9, Alpha=1, GravityOverride:Vector="0,0,-980", ExternalForce="0,0,0") -> (Pose) {
    angularDampingOverride=1
    angularSpringConstant=500
    bAngularSpring=true
    bOverrideAngularDamping=true
    bUseGravityOverride=true
    gravityScale=0.35
    simulationSpace=RootRelative
    alphaBoolBlend {blendOption=Linear}
    boundBone {boneName=Dyn_pendant}
    componentAppliedLinearAccClamp=<6000 6000 6000>
    componentLinearAccScale=<1 1 1>
    componentLinearVelScale=<1 1 1>
    gravityOverride {z=-980}
    physicsBodyDefinitions [{
      boundBone {boneName=Dyn_pendant}
      boxExtents=<10 10 10>
      constraintSetup {angularLimitsMax=<5 10 75>, angularLimitsMin=<-5 -10 -5>, angularTarget=<1 0 0>}
      localJointOffset=<1.2 0 0>
     }]
   }
   n11 = "Constraint - Base_zipper (PPPPPPP)"!(
    ComponentPose=@n10
    Alpha=1
    ConstraintWeights_0=0.1069
    ConstraintWeights_1:float=0.0642
    ConstraintWeights_2:float=0.4889
    ConstraintWeights_3:float=0.0111
    ConstraintWeights_4:float=0.018
    ConstraintWeights_5:float=0.014
    ConstraintWeights_6:float=0.297
   ) -> (Pose) {
    alphaBoolBlend {blendOption=Linear}
    boneToModify {boneName=Base_zipper}
    constraintSetup [
     {targetBone {boneName=spine_05}}
     {targetBone {boneName=clavicle_l}}
     {targetBone {boneName=spine_04}}
     {targetBone {boneName=upperarm_twist_01_l}}
     {targetBone {boneName=spine_03}}
     {targetBone {boneName=spine_02}}
     {targetBone {boneName=deform_pec_l}}
    ]
    constraintWeights [1,1,1,1,1,1,1]
   }
   Output Pose(Result=Component To Local(ComponentPose=@n11)) {layerGroup=None}
  ]}
 EventGraph {nodes [
   Event BlueprintUpdateAnimation()
   > n1 = Branch(Condition=Get bIsInFrontEnd())
   > Set SimlationSpace(SimlationSpace:AnimPhysSimSpaceType Enum=World) -> (Output_Get:AnimPhysSimSpaceType Enum)
   > Set FRAlpha(FRAlpha:double=0.5) -> (Output_Get:double)
   Set SimlationSpace(execute=@n1.else, SimlationSpace=Component)
   > Set FRAlpha(FRAlpha=1)
   Event OnCharacterCustomizationLoaded()
   > ResetDynamics(InTeleportType=TeleportPhysics)
   > n2 = Cast To PlayerPawn_Athena(Object=TryGetPawnOwner())
   > Branch(Condition=CheckForTail!(self:Custom Character Part Anim Instance Object Reference) -> (bHasTail:bool))
   > n3 = Branch(Condition=CheckForBackbling().bIsBackblingLong)
   n4 = GetSkeletalMeshForPartType(self=@n2, PartType=MiscOrTail)
   > SetVisibility!(execute, self:Scene Component Object Reference=@n4, bNewVisibility:bool=false, bPropagateToChildren:bool=false) -> (then)
   SetVisibility(execute=@n3.else, self=@n4, bNewVisibility=true, bPropagateToChildren=false)
   Event OnParachuteClosed()
   > Set HangTwo arms(HangTwo arms:double=0) -> (Output_Get:double)
   Event OnParachuteOpened()
   > Switch on EGliderType(Selection=Get PlayerGliderType() -> (PlayerGliderType=HangGlider))
   > Set HangTwo arms(HangTwo arms=1)
  ]}
]
variables [
 FRAlpha {type=number}
 HangTwo arms {type=number}
 SimlationSpace {type=AnimPhysSimSpaceType}
]
