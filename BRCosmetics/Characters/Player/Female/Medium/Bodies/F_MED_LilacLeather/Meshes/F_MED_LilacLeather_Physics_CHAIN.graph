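# Physics asset for the dangling chain on the F_MED_LilacLeather female player body, covering the chain bones, pelvis and right leg.
# Front and back four-link capsule chains hang from base bodies constrained to the pelvis, with twist locked and narrow swing limits.
# A constraint joins the two end links into a loop, and each end link carries a second, larger capsule.
# A glute base body attaches to the right thigh with all angular motion free.
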PhysicsAsset /BRCosmetics/Characters/Player/Female/Medium/Bodies/F_MED_LilacLeather/Meshes/F_MED_LilacLeather_Physics_CHAIN
bodies {
 Base_chain_bk [{shapeType=Capsule, length=0.505, radius=0.505, shapeName=Base_chain_bk_capsule, rotation=<rot 0 -90 0>}]
 Base_chain_fr [{shapeType=Capsule, length=0.505, radius=0.505, shapeName=Base_chain_fr_capsule, rotation=<rot 0 -90 0>}]
 Base_chain_glute [{shapeType=Capsule, length=0.505, radius=0.505, shapeName=Base_chain_glute_capsule, rotation=<rot 0 -90 0>}]
 Dyn_chain_bk_01 [{
   shapeType=Capsule
   length=1.8491
   radius=1.3917
   shapeName=Dyn_chain_bk_01_capsule
   center=<1.593 0.0017 0.1128>
   rotation=<rot 0 90.2381 -85.2737>
  }]
 Dyn_chain_bk_02 [{
   shapeType=Capsule
   length=2.0453
   radius=1.3099
   shapeName=Dyn_chain_bk_02_capsule
   center=<1.5683 0.183 0.1092>
   rotation=<rot 0 89.992 -85.8321>
  }]
 Dyn_chain_bk_03 [{
   shapeType=Capsule
   length=1.9791
   radius=1.2615
   shapeName=Dyn_chain_bk_03_capsule
   center=<1.8348 0.2059 -1.0696>
   rotation=<rot 0 91.5156 -80.8787>
  }]
 Dyn_chain_bk_04 [
  {
   shapeType=Capsule
   length=2.4674
   radius=1.7487
   shapeName=Dyn_chain_bk_04_capsule
   center=<0.1799 0.7747 -0.6531>
   rotation=<rot 55.0215 -116.8184 -126.1604>
  }
  {
   shapeType=Capsule
   length=10
   radius=2
   shapeName=Dyn_chain_bk_04_capsule
   center=<6.6078 -3.7422 -1.2959>
   rotation=<rot -39.3361 54.4883 -93.9723>
  }
 ]
 Dyn_chain_fr_01 [{
   shapeType=Capsule
   length=1.8941
   radius=1.2893
   shapeName=Dyn_chain_fr_01_capsule
   center=<1.2002 0.2305 0.0935>
   rotation=<rot 0 88.3637 -87.1896>
  }]
 Dyn_chain_fr_02 [{
   shapeType=Capsule
   length=1.8169
   radius=1.2891
   shapeName=Dyn_chain_fr_02_capsule
   center=<1.6974 0.0001 0.0238>
   rotation=<rot 0 94.2348 -89.2018>
  }]
 Dyn_chain_fr_03 [{
   shapeType=Capsule
   length=3.1342
   radius=1.2379
   shapeName=Dyn_chain_fr_03_capsule
   center=<2.2576 -0.9409 -0.1346>
   rotation=<rot -20.614 -82.7786 81.8833>
  }]
 Dyn_chain_fr_04 [
  {
   shapeType=Capsule
   length=3
   radius=2.0065
   shapeName=Dyn_chain_fr_04_capsule
   center=<0.9725 -1.385 0.4653>
   rotation=<rot -2.7045 98.0291 -62.906>
  }
  {
   shapeType=Capsule
   length=7
   radius=2
   shapeName=Dyn_chain_fr_04_capsule
   center=<3.1014 -0.2308 -5.6034>
   rotation=<rot 28.1219 -19.9047 -20.8447>
  }
 ]
 calf_r [{shapeType=Capsule, length=23.9219, radius=7.1011, center=<21.6297 1.8044 0.4941>, rotation=<rot 0 92.3606 -88.8855>}]
 foot_r [{shapeType=Capsule, length=12.6309, radius=7.3215, center=<3.2124 -8.3002 0.4338>, rotation=<rot 0 9.2657 -84.721>}]
 pelvis [{shapeType=Capsule, length=11.9255, radius=10.1811, center=<-1.3453 -0.9443 0>, rotation=<rot 0 0 -0.0243>}]
 thigh_r [{shapeType=Capsule, length=21.5802, radius=9.3821, center=<16.5148 1.5284 -1.0747>, rotation=<rot 0 -92.5786 -87.7498>}]
}
constraints [
 {bone1Name=thigh_r, bone2Name=pelvis, swing1LimitDegrees=5, swing2LimitDegrees=5, twistLimitDegrees=5}
 {bone1Name=calf_r, bone2Name=thigh_r, swing1LimitDegrees=45, swing2LimitDegrees=45, twistLimitDegrees=45}
 {bone1Name=foot_r, bone2Name=calf_r, swing1LimitDegrees=45, swing2LimitDegrees=45, twistLimitDegrees=45}
 {bone1Name=Base_chain_fr, bone2Name=pelvis, swing1LimitDegrees=11, swing2LimitDegrees=3, twistLimitDegrees=20, twistMotion=Locked}
 {bone1Name=Dyn_chain_fr_01, bone2Name=Base_chain_fr, swing1LimitDegrees=11, swing2LimitDegrees=11, twistLimitDegrees=20, twistMotion=Locked}
 {bone1Name=Dyn_chain_fr_02, bone2Name=Dyn_chain_fr_01, swing1LimitDegrees=11, swing2LimitDegrees=11, twistLimitDegrees=20, twistMotion=Locked}
 {bone1Name=Dyn_chain_fr_03, bone2Name=Dyn_chain_fr_02, swing1LimitDegrees=11, swing2LimitDegrees=3, twistLimitDegrees=20, twistMotion=Locked}
 {bone1Name=Dyn_chain_fr_04, bone2Name=Dyn_chain_fr_03, swing1LimitDegrees=11, swing2LimitDegrees=3, twistLimitDegrees=20, twistMotion=Locked}
 {bone1Name=Base_chain_bk, bone2Name=pelvis, swing1LimitDegrees=11, swing2LimitDegrees=3, twistLimitDegrees=20, twistMotion=Locked}
 {bone1Name=Dyn_chain_bk_01, bone2Name=Base_chain_bk, swing1LimitDegrees=11, swing2LimitDegrees=11, twistLimitDegrees=20, twistMotion=Locked}
 {bone1Name=Dyn_chain_bk_02, bone2Name=Dyn_chain_bk_01, swing1LimitDegrees=11, swing2LimitDegrees=11, twistLimitDegrees=20, twistMotion=Locked}
 {bone1Name=Dyn_chain_bk_03, bone2Name=Dyn_chain_bk_02, swing1LimitDegrees=11, swing2LimitDegrees=3, twistLimitDegrees=20, twistMotion=Locked}
 {bone1Name=Dyn_chain_bk_04, bone2Name=Dyn_chain_bk_03, swing1LimitDegrees=11, swing2LimitDegrees=11, twistLimitDegrees=20, twistMotion=Locked}
 {
  bone1Name=Base_chain_glute
  bone2Name=thigh_r
  swing1LimitDegrees=45
  swing1Motion=Free
  swing2LimitDegrees=45
  swing2Motion=Free
  twistLimitDegrees=45
  twistMotion=Free
 }
 {bone1Name=Dyn_chain_fr_04, bone2Name=Dyn_chain_bk_04, swing1LimitDegrees=11, swing2LimitDegrees=11, twistLimitDegrees=45, twistMotion=Locked}
]
