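# Post-copy AnimBP for the LilacLeather female face accessory mesh, adding secondary motion to a pendant and hair strands.
# It copies the parent mesh pose, then positions Base_Pendant by a weighted constraint to spine, clavicle, neck and jawline plus a yaw offset.
# Chained spring-driven AnimDynamics nodes then simulate the pendant and four hair bones root-relative.
# The event graph sets SimlationSpace from a front-end check, but no dynamics node reads it.
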AnimBlueprint /BRCosmetics/Characters/Player/Female/Medium/Bodies/F_MED_LilacLeather/Meshes/Parts/F_MED_LilacLeather_FaceAcc_AnimBP
target_skeleton=&/F_MED_LilacLeather_FaceAcc_Skeleton
graphs [
 AnimGraph {nodes [
   "Constraint - Base_Pendant (PPPPP)"!(
    ComponentPose
    Alpha=1
    ConstraintWeights_0=0.7213
    ConstraintWeights_1:float=0
    ConstraintWeights_2:float=0.1992
    ConstraintWeights_3:float=0.0546
    ConstraintWeights_4:float=0.0248
   ) -> (Pose) {
    alphaBoolBlend {blendOption=Linear}
    boneToModify {boneName=Base_Pendant}
    constraintSetup [
     {targetBone {boneName=spine_05}}
     {targetBone {boneName=clavicle_r}}
     {targetBone {boneName=neck_01}}
     {targetBone {boneName=neck_02}}
     {targetBone {boneName=FACIAL_C_Jawline}}
    ]
    constraintWeights [1,1,1,1,1]
   }
   n1 = "Constraint - Base_Pendant (PPPPP)"(
    ComponentPose=Local To Component(LocalPose=Copy Pose From Mesh() {bUseAttachedParent=true})
    Alpha=1
    ConstraintWeights_0=0.7213
    ConstraintWeights_1=0
    ConstraintWeights_2=0.1992
    ConstraintWeights_3=0.0546
    ConstraintWeights_4=0.0248
   ) {
    alphaBoolBlend {blendOption=Linear}
    boneToModify {boneName=Base_Pendant}
    constraintSetup [
     {targetBone {boneName=spine_05}}
     {targetBone {boneName=clavicle_r}}
     {targetBone {boneName=neck_01}}
     {targetBone {boneName=neck_02}}
     {targetBone {boneName=FACIAL_C_Jawline}}
    ]
    constraintWeights [1,1,1,1,1]
   }
   n2 = "Transform (Modify) Bone - Bone: Base_Pendant"!(ComponentPose=@n1, Alpha=1, Rotation="0,10,0", Scale="1,1,1") -> (Pose) {
    rotationMode=BMM_Additive
    rotationSpace=BCS_BoneSpace
    alphaBoolBlend {blendOption=Linear}
    boneToModify {boneName=Base_Pendant}
    rotation {yaw=10}
   }
   n3 = Anim Dynamics - Bone: Dyn_pendant!(SimulationSpace:AnimPhysSimSpaceType Enum, ComponentPose=@n2, Alpha=1, GravityOverride:Vector="0,0,-980", ExternalForce="0,0,0") -> (Pose) {
    angularDampingOverride=1
    angularSpringConstant=500
    bAngularSpring=true
    bOverrideAngularDamping=true
    bUseGravityOverride=true
    gravityScale=0.35
    simulationSpace=RootRelative
    alphaBoolBlend {blendOption=Linear}
    boundBone {boneName=Dyn_pendant}
    componentAppliedLinearAccClamp=<6000 6000 6000>
    componentLinearAccScale=<1 1 1>
    componentLinearVelScale=<1 1 1>
    gravityOverride {z=-980}
    physicsBodyDefinitions [{
      boundBone {boneName=Dyn_pendant}
      boxExtents=<10 10 10>
      constraintSetup {angularLimitsMax=<2 7 75>, angularLimitsMin=<-2 -7 -1>, angularTarget=<1 0 0>}
      localJointOffset=<0.9 0 0>
     }]
   }
   n4 = Anim Dynamics - Bone: Dyn_hair_fr_l1!(SimulationSpace:AnimPhysSimSpaceType Enum, ComponentPose=@n3, Alpha=1, ExternalForce="0,0,0") -> (Pose) {
    angularDampingOverride=0.9999
    angularSpringConstant=200
    bAngularSpring=true
    bOverrideAngularDamping=true
    bUseGravityOverride=true
    bUsePlanarLimit=false
    bUseSphericalLimits=true
    linearDampingOverride=0.9999
    linearSpringConstant=500
    simulationSpace=RootRelative
    alphaBoolBlend {blendOption=Linear}
    boundBone {boneName=Dyn_hair_fr_l1}
    componentAppliedLinearAccClamp=<5000 5000 5000>
    componentLinearAccScale=<0.5 0.5 0.5>
    componentLinearVelScale=<0.5 0.5 0.5>
    gravityOverride {z=-980}
    physicsBodyDefinitions [{
      collisionType=CustomSphere
      sphereCollisionRadius=1.4
      boundBone {boneName=Dyn_hair_fr_l1}
      boxExtents=<12 12 8>
      constraintSetup {angularLimitsMax=<0 5 15>, angularLimitsMin=<0 -30 -15>, angularTarget=<1 0 0>}
      localJointOffset=<2 0 0>
     }]
   }
   n5 = Anim Dynamics - Bone: Dyn_hair_side_l1!(SimulationSpace:AnimPhysSimSpaceType Enum, ComponentPose=@n4, Alpha=1, ExternalForce="0,0,0") -> (Pose) {
    angularDampingOverride=0.9999
    angularSpringConstant=200
    bAngularSpring=true
    bOverrideAngularDamping=true
    bUseGravityOverride=true
    bUsePlanarLimit=false
    bUseSphericalLimits=true
    linearDampingOverride=0.9999
    linearSpringConstant=500
    simulationSpace=RootRelative
    alphaBoolBlend {blendOption=Linear}
    boundBone {boneName=Dyn_hair_side_l1}
    componentAppliedLinearAccClamp=<5000 5000 5000>
    componentLinearAccScale=<0.5 0.5 0.5>
    componentLinearVelScale=<0.5 0.5 0.5>
    gravityOverride {z=-980}
    physicsBodyDefinitions [{
      collisionType=CustomSphere
      sphereCollisionRadius=1.4
      boundBone {boneName=Dyn_hair_side_l1}
      boxExtents=<12 12 8>
      constraintSetup {angularLimitsMax=<0 5 15>, angularLimitsMin=<0 -30 -15>, angularTarget=<1 0 0>}
      localJointOffset=<2 0 0>
     }]
   }
   n6 = Anim Dynamics - Bone: Dyn_hair_fr_r1!(SimulationSpace:AnimPhysSimSpaceType Enum, ComponentPose=@n5, Alpha=1, ExternalForce="0,0,0") -> (Pose) {
    angularDampingOverride=0.9999
    angularSpringConstant=200
    bAngularSpring=true
    bOverrideAngularDamping=true
    bUseGravityOverride=true
    bUsePlanarLimit=false
    bUseSphericalLimits=true
    linearDampingOverride=0.9999
    linearSpringConstant=500
    simulationSpace=RootRelative
    alphaBoolBlend {blendOption=Linear}
    boundBone {boneName=Dyn_hair_fr_r1}
    componentAppliedLinearAccClamp=<5000 5000 5000>
    componentLinearAccScale=<0.5 0.5 0.5>
    componentLinearVelScale=<0.5 0.5 0.5>
    gravityOverride {z=-980}
    physicsBodyDefinitions [{
      collisionType=CustomSphere
      sphereCollisionRadius=1.4
      boundBone {boneName=Dyn_hair_fr_r1}
      boxExtents=<12 12 8>
      constraintSetup {angularLimitsMax=<0 17 15>, angularLimitsMin=<0 -5 -15>, angularTarget=<1 0 0>}
      localJointOffset=<0.8 0 0>
     }]
   }
   n7 = Anim Dynamics - Bone: Dyn_hair_sdie_r1!(SimulationSpace:AnimPhysSimSpaceType Enum, ComponentPose=@n6, Alpha=1, ExternalForce="0,0,0") -> (Pose) {
    angularDampingOverride=0.9999
    angularSpringConstant=200
    bAngularSpring=true
    bOverrideAngularDamping=true
    bUseGravityOverride=true
    bUsePlanarLimit=false
    bUseSphericalLimits=true
    linearDampingOverride=0.9999
    linearSpringConstant=500
    simulationSpace=RootRelative
    alphaBoolBlend {blendOption=Linear}
    boundBone {boneName=Dyn_hair_sdie_r1}
    componentAppliedLinearAccClamp=<5000 5000 5000>
    componentLinearAccScale=<0.5 0.5 0.5>
    componentLinearVelScale=<0.5 0.5 0.5>
    gravityOverride {z=-980}
    physicsBodyDefinitions [{
      collisionType=CustomSphere
      sphereCollisionRadius=1.4
      boundBone {boneName=Dyn_hair_sdie_r1}
      boxExtents=<12 12 8>
      constraintSetup {angularLimitsMax=<0 22 15>, angularLimitsMin=<0 -5 -15>, angularTarget=<1 0 0>}
      localJointOffset=<1 0 0>
     }]
   }
   Output Pose(Result=Component To Local(ComponentPose=@n7)) {layerGroup=None}
  ]}
 EventGraph {nodes [
   Event BlueprintUpdateAnimation()
   > n1 = Branch(Condition=Get bIsInFrontEnd())
   > Set SimlationSpace(SimlationSpace:AnimPhysSimSpaceType Enum=World) -> (Output_Get:AnimPhysSimSpaceType Enum)
   Set SimlationSpace(execute=@n1.else)
  ]}
]
variables [SimlationSpace {type=AnimPhysSimSpaceType}]
